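; BP_AthenaHUD_UIDirectorComponent is a UI director component that manages the Athena Indicators scene for the dynamic Athena HUD.
; It adds the scene stored in its Athena Indicators Scene variable when the ready-to-show-UI event fires.
; On EndPlay it removes that same scene, whatever the end reason, so the scene's presence follows the component's lifetime.
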(event EventOnReadyToShowUI
  (CallFunction|AddScene (Variables|Default|GetAthena%20Indicators%20Scene)))

(event EventEndPlay
  (params (EndPlayReason :type EEndPlayReason%20Enum :default "Destroyed"))
  (CallFunction|RemoveScene (Variables|Default|GetAthena%20Indicators%20Scene)))
