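; BP_DynamicAthenaHudDirector is a Dynamic UI Director Blueprint that chooses where the Watcher icon sits in the Athena HUD.
; On BeginPlay it reads the Athena.UI.DrawWatcherIconInsideMiniMap console variable and checks whether the game is running as a mobile game.
; It adds the WatcherInMap scene when the variable is true on non-mobile, otherwise the WatcherDefault scene.
; The choice is made once at startup; Tick and the construction script are empty.
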(event EventBeginPlay
  (bind _returnvalue (Development|GetConsoleVariableBoolValue "Athena.UI.DrawWatcherIconInsideMiniMap"))
  ;[0 0 0] This will instantiate the Watcher's icon position regarding the CVar value
  (DynamicUIDirector|AddScene (select (and _returnvalue (not (TouchPlatform|IsMobileGame))) "/DynamicAthenaHUD/Scenes/AthenaHUD_WatcherInMap.AthenaHUD_WatcherInMap" "/DynamicAthenaHUD/Scenes/AthenaHUD_WatcherDefault.AthenaHUD_WatcherDefault")))

(event EventTick
  (params (DeltaSeconds :type Float%20%28single-precision%29)))

(fn ConstructionScript)
